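# Niagara vehicle cosmetic drift trail "SmokeRun": colored smoke sprites stream from the wheels, with mesh wraps on two of them.
# Two near-identical sprite emitters, one GPU-simulated, spawn at partition-indexed skeletal mesh sockets, with lifetime, size and velocity switched by GameModeID.
# The wrap emitter re-samples sockets each update and scales per wheel; all emitters mask per tire through FN_DisableTireFX.
# Color is either the painted variant or a time-cycled rainbow curve, chosen by User.bIsPainted.
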NiagaraSystem /VehicleCosmetics/DriftTrail/DriftTrail_SmokeRun/FX/NS_DriftTrail_SmokeRun
dependencies {
 usedDataInterfaces [
  &NiagaraDataInterfaceVectorCurve
  &NiagaraDataInterfaceColorCurve
  &NiagaraDataInterfaceCurve
  &NiagaraDataInterfaceVector2DCurve
  &NiagaraDataInterfaceSkeletalMesh
 ]
 usedDynamicInputs []
 usedModules [
  &self:SystemSpawnScript.SystemScriptSource.SystemScriptGraph.NiagaraNodeAssignment_0.SetVariables_#1
  &FN_UnpackVehicleFlags
  /VehicleCosmetics/FX/Niagara/Modules/VC_SetVehicleSpeedAttribs
  &SystemState
 ]
}
emitters {
 NE_VC_DriftTrail_Sprite_Parent {
  emitterSpawnScript [/SetVariables(Emitter.PowerSmoke:NiagaraFloat=&RandomRangeFloat(Minimum=0.8, Maximum=1.25)) {isSetParameters=true, module=SetVariables_#2}]
  emitterUpdateScript [
   &EmitterState()
   &SpawnRate(SpawnRate=200, Use Spawn Probability=false)
  ]
  particleSpawnScript [
   &VC_PaintedVariantsOverride(
    BurntSienna=<rgba 0.0642 0.0198 0.0079 1>
    R1 Black:LinearColor=<rgba 0.0208 0.0208 0.0208 1.5>
    R1 BurntSienna:LinearColor=<rgba 0.08 0.0239 0.0099 1.5>
    R1 Cobalt:LinearColor=<rgba 0.0195 0.0804 0.75 1>
    R1 Crimson:LinearColor=<rgba 3 0.1445 0.0507 1.5>
    R1 ForestGreen:LinearColor=<rgba 0 0.35 0.0592 1.5>
    R1 Grey:LinearColor=<rgba 0.1523 0.1768 0.25 1.25>
    R1 Lime:LinearColor=<rgba 0.4998 0.9375 0 1.35>
    R1 Orange:LinearColor=<rgba 0.6042 0.1813 0 1.5>
    R1 Pink:LinearColor=<rgba 0.6339 0.1819 0.776 1.5>
    R1 Purple:LinearColor=<rgba 0.103 0 0.4323 1.5>
    R1 Saffron:LinearColor=<rgba 0.7552 0.579 0 2>
    R1 SkyBlue:LinearColor=<rgba 0 0.287 0.8087 1.5>
    R1 TitaniumWhite:LinearColor=<rgba 0.9167 0.9167 0.9167 1.5>
    Use R1 Painted Overrides=true
   )
   &InitializeParticle(
    Color=$StackContext.Color
    Lifetime=&MergeHelper_FloatSwitch(0=1.5, 1=1.25, 2=0.2, 3=1.5, Index=$System.GameModeID, Num=::NewEnumerator2)
    Sprite Rotation Mode=::NewEnumerator2
    Sprite Size Mode=::NewEnumerator3
    Uniform Sprite Size=&MergeHelper_FloatSwitch(
     0=&RandomRangeFloat(Minimum=40, Maximum=45)
     1=&RandomRangeFloat(Minimum=60, Maximum=75)
     2=&RandomRangeFloat(Minimum=80, Maximum=95)
     3=&RandomRangeFloat(Minimum=40, Maximum=45)
     Index=$System.GameModeID
     Num=::NewEnumerator2
    )
    UsePositionOffset=false
    Write Material Random=true
   )
   &PartitionParticles()
   &SkeletalMeshLocation(
    Bone / Socket Index=&Add_Integer(A=$Particles.PartitionParticles.Partition, B=4)
    Bone / Socket Sampling Mode=::NewEnumerator5
    MaskSpawnGroup=false
    Mesh Sampling Type=::NewEnumerator6
    Sampled Position Offset=&AddVectorToPosition(Position=<-30 0 -20>, Vector=$User.offset)
    Skeletal Mesh=$System.SkeletalMesh
   )
   &AddVelocity(Velocity=&MergeHelper_VectorSwitch(0=<-400 0 0>, 1=<-200 0 0>, 2=<-100 0 0>, 3=<-400 0 0>, Index=$System.GameModeID, Num=::NewEnumerator2))
   /VehicleCosmetics/FX/VC_FXParents/Modules/DLR_ProductPreviews(GameModeID:NiagaraInt32=$System.GameModeID, bPreviewVehicleSpeed:NiagaraBool=$System.bPreviewVehicleSpeed)
  ]
  particleUpdateScript [
   &FN_DisableTireFX(
    Mask Index Attribute:NiagaraInt32=$Particles.PartitionParticles.Partition
    Mask Index FL=2
    Mask Index FR=3
    Mask Index RL=0
    Mask Index RR=1
   )
   &ParticleState()
   &SpriteRotationRate(Rotation Rate=&RandomRangeFloat(Minimum=0, Maximum=1, Evaluation Type=::NewEnumerator1))
   &ScaleSpriteSize(
    Scale Factor=&MakeVector2D(
     X=&FloatFromCurve(
      FloatCurve=DataInterface() {propertyValues {Curve {
         defaultValue=2
         keys [
          {arriveTangent=0.0059, interpMode=RCIM_Cubic, leaveTangent=0.0059, tangentMode=RCTM_User, time=0, value=1}
          {arriveTangent=2.1735, interpMode=RCIM_Cubic, leaveTangent=2.1735, tangentMode=RCTM_User, time=0.2, value=1.2027}
          {arriveTangent=-0.0358, interpMode=RCIM_Cubic, leaveTangent=-0.0358, tangentMode=RCTM_User, time=0.5, value=1.5}
          <at 0.8 1.25 RCIM_Cubic>
         ]
        }}}
      Scale Curve=0.8
     )
     Y=&FloatFromCurve(
      FloatCurve=DataInterface() {propertyValues {Curve {
         defaultValue=2
         keys [
          {arriveTangent=0.0059, interpMode=RCIM_Cubic, leaveTangent=0.0059, tangentMode=RCTM_User, time=0, value=1}
          {arriveTangent=2.1735, interpMode=RCIM_Cubic, leaveTangent=2.1735, tangentMode=RCTM_User, time=0.2, value=1.2027}
          {arriveTangent=-0.0358, interpMode=RCIM_Cubic, leaveTangent=-0.0358, tangentMode=RCTM_User, time=0.5, value=1.5}
          <at 0.8 1.25 RCIM_Cubic>
         ]
        }}}
      Scale Curve=1.2
     )
    )
    Scale Sprite Size Mode=::NewEnumerator2
   )
   &InheritSourceMovement(
    Applied Position Scale=&VectorFromFloat(Value=&FloatFromCurve(FloatCurve=DataInterface() {propertyValues {Curve {
         defaultValue=0
         keys [{interpMode=RCIM_Cubic, tangentMode=RCTM_User, time=0.1, value=1}, {interpMode=RCIM_Cubic, tangentMode=RCTM_User, time=1, value=0.4}]
        }}}))
    ApplyForce=false
    ApplyPosition=true
    ApplyPositionFalloff=false
   ) {enabled=false}
   &AccelerationForce(Acceleration=<0 0 15>)
   &InheritSourceMovement(
    Applied Position Falloff Distance=650
    Applied Position Scale=&Multiply_VectorByFloat(
     Vector=<1 1 1>
     Float=&FloatFromCurve(FloatCurve=DataInterface() {propertyValues {Curve {
         defaultValue=0
         keys [
          {arriveTangent=-0.0976, interpMode=RCIM_Cubic, leaveTangent=-0.0976, tangentMode=RCTM_User, time=0, value=1}
          {interpMode=RCIM_Cubic, tangentMode=RCTM_User, time=1, value=0}
         ]
        }}})
    )
    ApplyForce=false
    ApplyPosition=true
    ApplyPositionFalloff=false
   ) {module=InheritSourceMovement001}
   &DynamicMaterialParameters(
    Index 0 Param 2=8
    Index 0 Param 3=&FloatFromCurve(FloatCurve=DataInterface() {propertyValues {Curve {
        defaultValue=1
        keys [
         {arriveTangent=4.1785, interpMode=RCIM_Cubic, leaveTangent=4.1785, tangentMode=RCTM_User, time=0, value=1}
         {arriveTangent=0.0224, interpMode=RCIM_Cubic, leaveTangent=0.0224, tangentMode=RCTM_User, time=1, value=64}
        ]
       }}})
    Module.Index0Float0Write=true
    Module.Index0Float1Write=true
    Module.Index0Float2Write=true
    Module.Index0Float3Write=true
   )
   &ScaleColor(
    Scale Alpha=&FloatFromCurve(
     FloatCurve=DataInterface() {propertyValues {Curve {keys [
         {arriveTangent=2.3917, interpMode=RCIM_Cubic, leaveTangent=2.3917, tangentMode=RCTM_User, time=0, value=0.5}
         {arriveTangent=1.25, time=0.2, value=1.5}
         {arriveTangent=0.5269, interpMode=RCIM_Cubic, leaveTangent=0.5269, tangentMode=RCTM_User, time=0.5, value=1.5}
         {arriveTangent=-12.9999, interpMode=RCIM_Cubic, leaveTangent=-12.9999, tangentMode=RCTM_User, time=0.9, value=0}
        ]}}}
     Scale Curve=0.2
    )
    Scale RGB=&VectorFromFloat(Value=1)
    ScaleA=true
    ScaleRGB=true
   ) {enabled=false}
   /SetVariables(Particles.R1ColorPaintds:LinearColor=&MakeCustomLinearColorFromBool(
     Bool=$User.bIsPainted
     True Color=&MultiplyLinearColor(
      Linear Color=$Particles.Initial.R1 Color
      Scale Factor=&MakeVector4FromVectorAndFloat(
       "Vector (RGB)"=&VectorFromFloat(Value=&FloatFromCurve(
         FloatCurve=DataInterface() {propertyValues {Curve {
            defaultValue=0
            keys [
             {arriveTangent=-0.0014, interpMode=RCIM_Cubic, tangentMode=RCTM_User, time=0, value=0.8}
             {arriveTangent=0.0002, interpMode=RCIM_Cubic, leaveTangent=0.0002, tangentMode=RCTM_User, time=1, value=0.8}
            ]
           }}}
         Scale Curve=&SelectFloatFromArray(
          Direct Array Index=$System.ColorVariantID
          Float Selection Array=DataInterface() {propertyValues {FloatData [1,1,1,1.5,1,0.2,1,1,1,1,1,1,1,1]}}
         )
        ))
       "Float (Alpha)"=&FloatFromCurve(
        FloatCurve=DataInterface() {propertyValues {Curve {keys [
            {arriveTangent=-0.1172, interpMode=RCIM_Cubic, leaveTangent=-0.1172, tangentMode=RCTM_User, time=0.5, value=1}
            {arriveTangent=-0.1107, interpMode=RCIM_Cubic, leaveTangent=-0.1107, tangentMode=RCTM_User, time=0.95, value=0.02}
           ]}}}
        Scale Curve=&SelectFloatFromArray(Direct Array Index=$System.ColorVariantID, Float Selection Array=DataInterface() {propertyValues {FloatData [1,1,1,1,1,1,1,1,1,1,1,1,1,1]}})
       )
      )
     )
     False Color=&Multiply_LinearColorByFloat(
      Linear Color=&MultiplyLinearColor(
       Linear Color=&ColorFromCurve(
        ColorCurve=DataInterface() {propertyValues {
          ExposedName=RainbowLUT
          bExposeCurve=true
          AlphaCurve {keys [<at 0 1>,<at 1 1>]}
          BlueCurve {keys [<at 0.029 0.1119>,<at 0.0921 0.1038>,<at 0.1451 0.0904>,<at 0.2032 0.1492>,<at 0.2663 0.0817>,<at 0.3395 0.2932>,<at 0.4077 0.6979>,<at 0.5011 0.3464>,<at 0.585 0.5365>,<at 0.6601 0.6458>,<at 0.7409 0.7>,<at 0.8166 0.4948>,<at 0.91 0.4663>,<at 0.9655 0.1126>]}
          GreenCurve {keys [<at 0.029 0.0648>,<at 0.0921 0.2309>,<at 0.1451 0.3964>,<at 0.2032 0.6927>,<at 0.2663 0.3646>,<at 0.3395 0.4167>,<at 0.4077 0.3847>,<at 0.5011 0.1064>,<at 0.585 0.0307>,<at 0.6601 0.0227>,<at 0.7409 0>,<at 0.8166 0.0464>,<at 0.91 0.0984>,<at 0.9655 0.0655>]}
          RedCurve {keys [<at 0.029 0.4852>,<at 0.0921 0.6042>,<at 0.1451 0.526>,<at 0.2032 0.4557>,<at 0.2663 0.1001>,<at 0.3395 0.0961>,<at 0.4077 0.12>,<at 0.5011 0.0215>,<at 0.585 0.0435>,<at 0.6601 0.1361>,<at 0.7409 0.4088>,<at 0.8166 0.3273>,<at 0.91 0.6094>,<at 0.9655 0.4849>]}
         }}
        CurveIndex=&Frac_Float(Float=&Multiply_Float(A=$Engine.Time, B=0.095))
       )
       Scale Factor=&MakeVector4FromVectorAndFloat(
        "Vector (RGB)"=&VectorFromFloat(Value=&FloatFromCurve(
          FloatCurve=DataInterface() {propertyValues {Curve {
             defaultValue=0.5
             keys [{interpMode=RCIM_Cubic, tangentMode=RCTM_User, time=0, value=0.8}, {interpMode=RCIM_Cubic, tangentMode=RCTM_User, time=1, value=1}]
            }}}
          Scale Curve=0.95
         ))
        "Float (Alpha)"=&FloatFromCurve(
         FloatCurve=DataInterface() {propertyValues {Curve {keys [
             {arriveTangent=0.0379, interpMode=RCIM_Cubic, leaveTangent=0.0379, tangentMode=RCTM_User, time=0, value=1}
             {arriveTangent=-0.1314, interpMode=RCIM_Cubic, leaveTangent=-0.1314, tangentMode=RCTM_User, time=0.5, value=1}
             {arriveTangent=-0.1592, interpMode=RCIM_Cubic, leaveTangent=-0.1592, tangentMode=RCTM_User, time=0.95, value=0.02}
            ]}}}
         Scale Curve=&RandomRangeFloat(Minimum=0.8, Maximum=1.5)
        )
       )
      )
      Float=1
     )
    )) {isSetParameters=true, module=SetVariables_#3}
   &SolveForcesAndVelocity(Clamp Velocity=false, Limit Acceleration=false)
  ]
  rendererClasses [&NiagaraSpriteRendererProperties]
  renderers [
   {rendererClass=&NiagaraSpriteRendererProperties}
   {rendererClass=&NiagaraSpriteRendererProperties, rendererIndex=1}
  ]
 }
 NE_VC_DriftTrail_Sprite_Parent001 {
  simTarget=GPUComputeSim
  emitterSpawnScript [/SetVariables(Emitter.PowerSmoke:NiagaraFloat=&RandomRangeFloat(Minimum=0.8, Maximum=1.25)) {isSetParameters=true, module=SetVariables_#2}]
  emitterUpdateScript [
   &EmitterState()
   &SpawnRate(SpawnRate=166, Use Spawn Probability=false)
  ]
  particleSpawnScript [
   &VC_PaintedVariantsOverride(
    BurntSienna=<rgba 0.0642 0.0198 0.0079 1>
    R1 Black:LinearColor=<rgba 0.0208 0.0208 0.0208 1.5>
    R1 BurntSienna:LinearColor=<rgba 0.08 0.0239 0.0099 1.5>
    R1 Cobalt:LinearColor=<rgba 0.0195 0.0804 0.75 1>
    R1 Crimson:LinearColor=<rgba 3 0.1445 0.0507 1.5>
    R1 ForestGreen:LinearColor=<rgba 0 0.35 0.0592 1.5>
    R1 Grey:LinearColor=<rgba 0.1523 0.1768 0.25 1.25>
    R1 Lime:LinearColor=<rgba 0.4998 0.9375 0 1.35>
    R1 Orange:LinearColor=<rgba 0.6042 0.1813 0 1.5>
    R1 Pink:LinearColor=<rgba 0.6339 0.1819 0.776 1.5>
    R1 Purple:LinearColor=<rgba 0.103 0 0.4323 1.5>
    R1 Saffron:LinearColor=<rgba 0.7552 0.579 0 2>
    R1 SkyBlue:LinearColor=<rgba 0 0.287 0.8087 1.5>
    R1 TitaniumWhite:LinearColor=<rgba 0.9167 0.9167 0.9167 1.5>
    Use R1 Painted Overrides=true
   )
   &InitializeParticle(
    Color=$StackContext.Color
    Lifetime=&MergeHelper_FloatSwitch(0=1.5, 1=1.25, 2=0.2, 3=1.5, Index=$System.GameModeID, Num=::NewEnumerator2)
    Sprite Rotation Mode=::NewEnumerator2
    Sprite Size Mode=::NewEnumerator3
    Uniform Sprite Size=&MergeHelper_FloatSwitch(
     0=&RandomRangeFloat(Minimum=40, Maximum=45)
     1=&RandomRangeFloat(Minimum=55, Maximum=60)
     2=&RandomRangeFloat(Minimum=80, Maximum=95)
     3=&RandomRangeFloat(Minimum=40, Maximum=45)
     Index=$System.GameModeID
     Num=::NewEnumerator2
    )
    UsePositionOffset=false
    Write Material Random=true
   )
   &PartitionParticles()
   &SkeletalMeshLocation(
    Bone / Socket Index=&Add_Integer(A=$Particles.PartitionParticles.Partition, B=4)
    Bone / Socket Sampling Mode=::NewEnumerator5
    MaskSpawnGroup=false
    Mesh Sampling Type=::NewEnumerator6
    Sampled Position Offset=&AddVectorToPosition(Position=<-30 0 -20>, Vector=$User.offset)
    Skeletal Mesh=$System.SkeletalMesh
   )
   &AddVelocity(Velocity=&MergeHelper_VectorSwitch(0=<-400 0 0>, 1=<-200 0 0>, 2=<-100 0 0>, 3=<-400 0 0>, Index=$System.GameModeID, Num=::NewEnumerator2))
   &DLR_ProductPreviews(GameModeID:NiagaraInt32=$System.GameModeID, bPreviewVehicleSpeed:NiagaraBool=$System.bPreviewVehicleSpeed)
  ]
  particleUpdateScript [
   &FN_DisableTireFX(
    Mask Index Attribute:NiagaraInt32=$Particles.PartitionParticles.Partition
    Mask Index FL=2
    Mask Index FR=3
    Mask Index RL=0
    Mask Index RR=1
   )
   &ParticleState()
   &SpriteRotationRate(Rotation Rate=&RandomRangeFloat(Minimum=0, Maximum=1, Evaluation Type=::NewEnumerator1))
   &ScaleSpriteSize(
    Scale Factor=&MakeVector2D(
     X=&FloatFromCurve(FloatCurve=DataInterface() {propertyValues {Curve {
         defaultValue=2
         keys [
          {arriveTangent=0.0059, interpMode=RCIM_Cubic, leaveTangent=0.0059, tangentMode=RCTM_User, time=0, value=1}
          {arriveTangent=2.1735, interpMode=RCIM_Cubic, leaveTangent=2.1735, tangentMode=RCTM_User, time=0.2, value=1.2027}
          {arriveTangent=-0.0358, interpMode=RCIM_Cubic, leaveTangent=-0.0358, tangentMode=RCTM_User, time=0.5, value=1.5}
          <at 0.8 1.25 RCIM_Cubic>
         ]
        }}})
     Y=&FloatFromCurve(
      FloatCurve=DataInterface() {propertyValues {Curve {
         defaultValue=2
         keys [
          {arriveTangent=0.0059, interpMode=RCIM_Cubic, leaveTangent=0.0059, tangentMode=RCTM_User, time=0, value=1}
          {arriveTangent=2.1735, interpMode=RCIM_Cubic, leaveTangent=2.1735, tangentMode=RCTM_User, time=0.2, value=1.2027}
          {arriveTangent=-0.0358, interpMode=RCIM_Cubic, leaveTangent=-0.0358, tangentMode=RCTM_User, time=0.5, value=1.5}
          <at 0.8 1.25 RCIM_Cubic>
         ]
        }}}
      Scale Curve=1.2
     )
    )
    Scale Sprite Size Mode=::NewEnumerator2
   )
   &InheritSourceMovement(
    Applied Position Scale=&VectorFromFloat(Value=&FloatFromCurve(FloatCurve=DataInterface() {propertyValues {Curve {
         defaultValue=0
         keys [{interpMode=RCIM_Cubic, tangentMode=RCTM_User, time=0.1, value=1}, {interpMode=RCIM_Cubic, tangentMode=RCTM_User, time=1, value=0.4}]
        }}}))
    ApplyForce=false
    ApplyPosition=true
    ApplyPositionFalloff=false
   ) {enabled=false}
   &AccelerationForce(Acceleration=<0 0 15>)
   &InheritSourceMovement(
    Applied Position Falloff Distance=650
    Applied Position Scale=&Multiply_VectorByFloat(
     Vector=<1 1 1>
     Float=&FloatFromCurve(FloatCurve=DataInterface() {propertyValues {Curve {
         defaultValue=0
         keys [
          {arriveTangent=-0.0976, interpMode=RCIM_Cubic, leaveTangent=-0.0976, tangentMode=RCTM_User, time=0, value=1}
          {interpMode=RCIM_Cubic, tangentMode=RCTM_User, time=1, value=0}
         ]
        }}})
    )
    ApplyForce=false
    ApplyPosition=true
    ApplyPositionFalloff=false
   ) {module=InheritSourceMovement001}
   &DynamicMaterialParameters(
    Index 0 Param 2=8
    Index 0 Param 3=&FloatFromCurve(FloatCurve=DataInterface() {propertyValues {Curve {
        defaultValue=1
        keys [
         {arriveTangent=4.1785, interpMode=RCIM_Cubic, leaveTangent=4.1785, tangentMode=RCTM_User, time=0, value=1}
         {arriveTangent=0.0224, interpMode=RCIM_Cubic, leaveTangent=0.0224, tangentMode=RCTM_User, time=1, value=64}
        ]
       }}})
    Module.Index0Float0Write=true
    Module.Index0Float1Write=true
    Module.Index0Float2Write=true
    Module.Index0Float3Write=true
   )
   &ScaleColor(
    Scale Alpha=&FloatFromCurve(
     FloatCurve=DataInterface() {propertyValues {Curve {keys [
         {arriveTangent=2.3917, interpMode=RCIM_Cubic, leaveTangent=2.3917, tangentMode=RCTM_User, time=0, value=0.5}
         {arriveTangent=1.25, time=0.2, value=1.5}
         {arriveTangent=0.5269, interpMode=RCIM_Cubic, leaveTangent=0.5269, tangentMode=RCTM_User, time=0.5, value=1.5}
         {arriveTangent=-12.9999, interpMode=RCIM_Cubic, leaveTangent=-12.9999, tangentMode=RCTM_User, time=0.9, value=0}
        ]}}}
     Scale Curve=0.2
    )
    Scale RGB=&VectorFromFloat(Value=1)
    ScaleA=true
    ScaleRGB=true
   ) {enabled=false}
   /SetVariables(Particles.R1ColorPaintds:LinearColor=&MakeCustomLinearColorFromBool(
     Bool=$User.bIsPainted
     True Color=&MultiplyLinearColor(
      Linear Color=$Particles.Initial.R1 Color
      Scale Factor=&MakeVector4FromVectorAndFloat(
       "Vector (RGB)"=&VectorFromFloat(Value=&FloatFromCurve(
         FloatCurve=DataInterface() {propertyValues {Curve {
            defaultValue=0
            keys [
             {arriveTangent=-0.0014, interpMode=RCIM_Cubic, tangentMode=RCTM_User, time=0, value=0.8}
             {arriveTangent=0.0002, interpMode=RCIM_Cubic, leaveTangent=0.0002, tangentMode=RCTM_User, time=1, value=0.8}
            ]
           }}}
         Scale Curve=&SelectFloatFromArray(
          Direct Array Index=$System.ColorVariantID
          Float Selection Array=DataInterface() {propertyValues {FloatData [1,1,1,1.5,1,0.2,1,1,1,1,1,1,1,1]}}
         )
        ))
       "Float (Alpha)"=&FloatFromCurve(
        FloatCurve=DataInterface() {propertyValues {Curve {keys [
            {arriveTangent=-0.1172, interpMode=RCIM_Cubic, leaveTangent=-0.1172, tangentMode=RCTM_User, time=0.5, value=1}
            {arriveTangent=-0.1107, interpMode=RCIM_Cubic, leaveTangent=-0.1107, tangentMode=RCTM_User, time=0.95, value=0.02}
           ]}}}
        Scale Curve=&SelectFloatFromArray(Direct Array Index=$System.ColorVariantID, Float Selection Array=DataInterface() {propertyValues {FloatData [1,1,1,1,1,1,1,1,1,1,1,1,1,1]}})
       )
      )
     )
     False Color=&Multiply_LinearColorByFloat(
      Linear Color=&MultiplyLinearColor(
       Linear Color=&ColorFromCurve(
        ColorCurve=DataInterface() {propertyValues {
          ExposedName=RainbowLUT
          bExposeCurve=true
          AlphaCurve {keys [<at 0 1>,<at 1 1>]}
          BlueCurve {keys [<at 0.029 0.1119>,<at 0.0921 0.1038>,<at 0.1451 0.0904>,<at 0.2032 0.1492>,<at 0.2663 0.0817>,<at 0.3395 0.2932>,<at 0.4077 0.6979>,<at 0.5011 0.3464>,<at 0.585 0.5365>,<at 0.6601 0.6458>,<at 0.7409 0.7>,<at 0.8166 0.4948>,<at 0.91 0.4663>,<at 0.9655 0.1126>]}
          GreenCurve {keys [<at 0.029 0.0648>,<at 0.0921 0.2309>,<at 0.1451 0.3964>,<at 0.2032 0.6927>,<at 0.2663 0.3646>,<at 0.3395 0.4167>,<at 0.4077 0.3847>,<at 0.5011 0.1064>,<at 0.585 0.0307>,<at 0.6601 0.0227>,<at 0.7409 0>,<at 0.8166 0.0464>,<at 0.91 0.0984>,<at 0.9655 0.0655>]}
          RedCurve {keys [<at 0.029 0.4852>,<at 0.0921 0.6042>,<at 0.1451 0.526>,<at 0.2032 0.4557>,<at 0.2663 0.1001>,<at 0.3395 0.0961>,<at 0.4077 0.12>,<at 0.5011 0.0215>,<at 0.585 0.0435>,<at 0.6601 0.1361>,<at 0.7409 0.4088>,<at 0.8166 0.3273>,<at 0.91 0.6094>,<at 0.9655 0.4849>]}
         }}
        CurveIndex=&Frac_Float(Float=&Multiply_Float(A=$Engine.Time, B=0.095))
       )
       Scale Factor=&MakeVector4FromVectorAndFloat(
        "Vector (RGB)"=&VectorFromFloat(Value=&FloatFromCurve(
          FloatCurve=DataInterface() {propertyValues {Curve {
             defaultValue=0.5
             keys [{interpMode=RCIM_Cubic, tangentMode=RCTM_User, time=0, value=0.8}, {interpMode=RCIM_Cubic, tangentMode=RCTM_User, time=1, value=1}]
            }}}
          Scale Curve=0.95
         ))
        "Float (Alpha)"=&FloatFromCurve(
         FloatCurve=DataInterface() {propertyValues {Curve {keys [
             {arriveTangent=0.0379, interpMode=RCIM_Cubic, leaveTangent=0.0379, tangentMode=RCTM_User, time=0, value=1}
             {arriveTangent=-0.1314, interpMode=RCIM_Cubic, leaveTangent=-0.1314, tangentMode=RCTM_User, time=0.5, value=1}
             {arriveTangent=-0.1592, interpMode=RCIM_Cubic, leaveTangent=-0.1592, tangentMode=RCTM_User, time=0.95, value=0.02}
            ]}}}
         Scale Curve=&RandomRangeFloat(Minimum=0.8, Maximum=1.5)
        )
       )
      )
      Float=1
     )
    )) {isSetParameters=true, module=SetVariables_#3}
   &SolveForcesAndVelocity(Clamp Velocity=false, Limit Acceleration=false)
  ]
  rendererClasses [&NiagaraSpriteRendererProperties]
  renderers [
   {rendererClass=&NiagaraSpriteRendererProperties}
   {rendererClass=&NiagaraSpriteRendererProperties, rendererIndex=1}
  ]
 }
 NE_VC_WheelWrap004 {
  emitterUpdateScript [
   &EmitterState(
    Inactive Response=::NewEnumerator1
    Life Cycle Mode=::NewEnumerator1
    Loop Behavior=::NewEnumerator1
    Loop Duration Mode=::NewEnumerator1
    UseLoopDelay=false
   )
   &SpawnBurst_Instantaneous(Spawn Count=2, Use Spawn Probability=false, UseLoopCountLimit=false)
  ]
  particleSpawnScript [
   &VC_PaintedVariantsOverride(
    Black=<rgba 0.015 0.015 0.015 10>
    BurntSienna=<rgba 0.2396 0.057 0.0311 1>
    Cobalt=<rgba 0 0.1524 1.2 1>
    Crimson=<rgba 0.4635 0 0.0126 5>
    ForestGreen=<rgba 0 0.276 0.0139 5>
    Grey=<rgba 0.1514 0.1514 0.1615 5>
    Lime=<rgba 0.2599 0.4427 0 5>
    Orange=<rgba 0.6724 0.1426 0.0038 10>
    Pink=<rgba 0.7347 0.2348 0.7448 10>
    Purple=<rgba 0.1621 0.0606 0.8958 10>
    R1 Black:LinearColor=<rgba 0.02 0.02 0.02 25>
    R1 BurntSienna:LinearColor=<rgba 0.0521 0.0239 0.0168 10>
    R1 Cobalt:LinearColor=<rgba 0.0213 0.0515 0.1875 1>
    R1 Crimson:LinearColor=<rgba 0.2812 0.0136 0.0048 1>
    R1 ForestGreen:LinearColor=<rgba 0 0.1771 0.03 1.5>
    R1 Grey:LinearColor=<rgba 0.1259 0.1322 0.151 1.25>
    R1 Lime:LinearColor=<rgba 0.1844 0.2604 0.0977 1.5>
    R1 Orange:LinearColor=<rgba 0.2344 0.0996 0.0357 1.5>
    R1 Pink:LinearColor=<rgba 0.2611 0.1375 0.3 1.5>
    R1 Purple:LinearColor=<rgba 0.104 0.0599 0.2448 1.5>
    R1 Saffron:LinearColor=<rgba 0.2969 0.2363 0.0371 2>
    R1 SkyBlue:LinearColor=<rgba 0.0828 0.1329 0.224 1.5>
    R1 TitaniumWhite:LinearColor=<rgba 0.2292 0.2292 0.2292 1.5>
    Saffron=<rgba 0.8854 0.6788 0 5>
    SkyBlue=<rgba 0.0197 0.3277 0.6 5>
    TitaniumWhite=<rgba 0.5596 0.5596 0.5625 5>
    Use R1 Painted Overrides=true
   )
   &InitializeParticle(
    Color=$StackContext.Color
    Lifetime=5
    Mesh Scale Mode=::NewEnumerator4
    Sprite Rotation Mode=::NewEnumerator2
    Sprite Size Mode=::NewEnumerator3
    UsePositionOffset=false
    Write Material Random=true
   )
   &PartitionParticles()
   &SkeletalMeshLocation(
    Bone / Socket Index=&Add_Integer(A=$Particles.PartitionParticles.Partition, B=4)
    Bone / Socket Sampling Mode=::NewEnumerator5
    Interpolate Mesh Sampling=false
    MaskSpawnGroup=false
    Mesh Sampling Type=::NewEnumerator6
    Skeletal Mesh=$System.SkeletalMesh
   )
   &InitialMeshOrientation(Mesh Orientation Mode=::NewEnumerator7, Rotation=<0 0 0.5>, Use Rotation Vector=true)
   &Color(Color=<rgba 0 1 0 1>) {enabled=false}
  ]
  particleUpdateScript [
   &FN_DisableTireFX(
    Mask Index Attribute:NiagaraInt32=$Particles.PartitionParticles.Partition
    Mask Index FL=2
    Mask Index FR=3
    Mask Index RL=0
    Mask Index RR=1
   )
   &VC_WheelWrapScale(Wheel Index Attribute:NiagaraInt32=&Add_Integer(A=$Particles.PartitionParticles.Partition, B=2))
   &ParticleState(Kill Particles When Lifetime Has Elapsed=false, Loop Particles Lifetime=true)
   &UpdateMeshOrientation(Rotation Coordinate Space=::NewEnumerator2, Rotation Vector=<0 1 0>) {enabled=false}
   &SkeletalMeshLocation(
    Bone / Socket Index=&Add_Integer(A=$Particles.PartitionParticles.Partition, B=4)
    Bone / Socket Sampling Mode=::NewEnumerator5
    MaskSpawnGroup=false
    Mesh Sampling Type=::NewEnumerator6
    Skeletal Mesh=$System.SkeletalMesh
   ) {module=SkeletalMeshLocation001}
   &SolveForcesAndVelocity(Clamp Velocity=false, Limit Acceleration=false)
   &ScaleColor(
    Scale RGB=&VectorFromCurve(VectorCurve=DataInterface() {propertyValues {XCurve {keys [<at 0 1>,<at 1 2>]}, YCurve {keys [<at 0 1>,<at 1 2>]}, ZCurve {keys [<at 0 1>,<at 1 2>]}}})
    ScaleA=true
    ScaleRGB=true
   )
   /SetVariables(Particles.BlueColor:LinearColor=&MakeCustomLinearColorFromBool(
     Bool=$User.bIsPainted
     True Color=&MultiplyLinearColor(
      Linear Color=$Particles.Initial.R1 Color
      Scale Factor=&MakeVector4FromVectorAndFloat("Vector (RGB)"=&VectorFromFloat(Value=3), "Float (Alpha)"=1)
     )
     False Color=&Multiply_LinearColorByFloat(
      Linear Color=&MultiplyLinearColor(
       Linear Color=&ColorFromCurve(
        ColorCurve=DataInterface() {propertyValues {
          ExposedName=RainbowLUT
          bExposeCurve=true
          AlphaCurve {keys [<at 0 1>,<at 1 1>]}
          BlueCurve {keys [<at 0.029 0.1119>,<at 0.0921 0.1038>,<at 0.1451 0.0904>,<at 0.2032 0.1492>,<at 0.2663 0.0817>,<at 0.3395 0.2932>,<at 0.4077 0.6979>,<at 0.5011 0.3464>,<at 0.585 0.5365>,<at 0.6601 0.6458>,<at 0.7409 0.7>,<at 0.8166 0.4948>,<at 0.91 0.4663>,<at 0.9655 0.1126>]}
          GreenCurve {keys [<at 0.029 0.0648>,<at 0.0921 0.2309>,<at 0.1451 0.3964>,<at 0.2032 0.6927>,<at 0.2663 0.3646>,<at 0.3395 0.4167>,<at 0.4077 0.3847>,<at 0.5011 0.1064>,<at 0.585 0.0307>,<at 0.6601 0.0227>,<at 0.7409 0>,<at 0.8166 0.0464>,<at 0.91 0.0984>,<at 0.9655 0.0655>]}
          RedCurve {keys [<at 0.029 0.4852>,<at 0.0921 0.6042>,<at 0.1451 0.526>,<at 0.2032 0.4557>,<at 0.2663 0.1001>,<at 0.3395 0.0961>,<at 0.4077 0.12>,<at 0.5011 0.0215>,<at 0.585 0.0435>,<at 0.6601 0.1361>,<at 0.7409 0.4088>,<at 0.8166 0.3273>,<at 0.91 0.6094>,<at 0.9655 0.4849>]}
         }}
        CurveIndex=&Frac_Float(Float=&Multiply_Float(A=$Engine.Time, B=0.095))
       )
       Scale Factor=&MakeVector4FromVectorAndFloat("Vector (RGB)"=&VectorFromFloat(Value=0.8), "Float (Alpha)"=3.5)
      )
      Float=1
     )
    )) {isSetParameters=true}
  ]
  rendererClasses [&NiagaraMeshRendererProperties]
  renderers [
   {rendererClass=&NiagaraMeshRendererProperties, rendererIndex=0}
   {rendererClass=&NiagaraMeshRendererProperties, rendererIndex=1}
  ]
 }
}
summary {
 systemName=NS_DriftTrail_SmokeRun
 emitters [
  {emitterName=NE_VC_WheelWrap004, rendererClasses [&NiagaraMeshRendererProperties]}
  {emitterName=NE_VC_DriftTrail_Sprite_Parent, rendererClasses [&NiagaraSpriteRendererProperties]}
  {emitterName=NE_VC_DriftTrail_Sprite_Parent001, simTarget=GPUComputeSim, rendererClasses [&NiagaraSpriteRendererProperties]}
 ]
 userVariables [
  User.bIsFrontendPreview {type=NiagaraBool, defaultValue=true}
  User.bIsPainted {type=NiagaraBool}
  User.ColorVariantID {type=NiagaraInt32}
  User.GameModeID {type=NiagaraInt32}
  User.offset {type=Vector3f, description="offset from wheels(set through cosmetic template)", defaultValue {struct=Vector3f, value=<0 0 0>}}
  User.PaintedColor {type=LinearColor, description="", defaultValue {struct=LinearColor, value=<rgba 1 1 1 1>}}
  User.PaintedScalar {type=NiagaraFloat, defaultValue=1, description=""}
  User.VehicleFlags {type=NiagaraInt32, defaultValue=806289408}
  User.Wheel_Radius_Array {
   type=NiagaraDataInterfaceArrayFloat
   defaultValue {
    struct=NiagaraExt_VariableValue_DataInterface
    value {
     dataInterface=&self:NiagaraDataInterfaceArrayFloat_1
     dataInterfaceClass=&NiagaraDataInterfaceArrayFloat
    }
   }
  }
  User.Wheel_Width_Array {
   type=NiagaraDataInterfaceArrayFloat
   defaultValue {
    struct=NiagaraExt_VariableValue_DataInterface
    value {
     dataInterface=&self:NiagaraDataInterfaceArrayFloat_0
     dataInterfaceClass=&NiagaraDataInterfaceArrayFloat
    }
   }
  }
  User.WheelTouchingArray {
   type=NiagaraDataInterfaceArrayFloat
   defaultValue {
    struct=NiagaraExt_VariableValue_DataInterface
    value {
     dataInterface=&self:NiagaraDataInterfaceArrayFloat_2
     dataInterfaceClass=&NiagaraDataInterfaceArrayFloat
    }
   }
  }
 ]
}
